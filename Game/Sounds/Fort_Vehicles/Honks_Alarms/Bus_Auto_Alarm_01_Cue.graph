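# Sound cue for a bus auto alarm in the vehicle honks and alarms set, built on a single bus horn honk wave.
# The wave player feeds a Modulator that varies playback, with its upper pitch limit held at the original pitch.
# It borrows a demolition-hit attenuation asset and plays under the impacts sound class at reduced volume.
# A concurrency override allows one instance at a time and blocks new plays while it sounds.
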SoundCue /Game/Sounds/Fort_Vehicles/Honks_Alarms/Bus_Auto_Alarm_01_Cue
root=Modulator_0
nodes [
 {id=Modulator_0, children [WavePlayer_0], properties {pitchMax=1}}
 {id=WavePlayer_0, properties {soundWaveAssetPtr=&/Bus_Horn_Honk_01}}
]
properties {
 attenuationSettings=/Game/Sounds/Attenuation/Fort_Demolition_Hit_Medium_Sized_Object_Att
 bOverrideConcurrency=true
 maxDistance=4000
 soundClassObject=&Fort_Impacts
 volumeMultiplier=0.64
 concurrencyOverrides {maxCount=1, resolutionRule=PreventNew}
}
