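# Sound cue for a destroyed York vehicle, judging by its name, routed to the Fort_Impacts sound class.
# A Mixer layers a hard car impact wave with the York car destroy wave, which enters after a short fixed Delay.
# It uses the demolition attenuation for medium-sized destroyed objects.
# A concurrency override allows only one instance at a time and prevents new ones from starting.
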SoundCue /Game/Sounds/Fort_Vehicles/Honks_Alarms/York_Vehicle_Destroyed_Horn_Cue
root=Mixer_0
nodes [
 {id=Delay_0, children [WavePlayer_0], properties {delayMax=0.1, delayMin=0.1}}
 {id=Mixer_0, children [WavePlayer_3,Delay_0], properties {inputVolume [1,1]}}
 {id=WavePlayer_0, properties {soundWaveAssetPtr=&/York_Car_Destroy}}
 {id=WavePlayer_3, properties {soundWaveAssetPtr=/Game/Sounds/Fort_Impact_Sounds/Demolition/Car/Car_Impact_Hard1}}
]
properties {
 attenuationSettings=/Game/Sounds/Attenuation/Fort_Demolition_Destroy_Medium_Sized_Object_Att
 bOverrideConcurrency=true
 maxDistance=4000
 soundClassObject=&Fort_Impacts
 volumeMultiplier=1
 concurrencyOverrides {maxCount=1, resolutionRule=PreventNew}
}
