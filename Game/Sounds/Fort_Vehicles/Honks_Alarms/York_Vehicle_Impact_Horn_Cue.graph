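# Sound cue for York vehicle impacts, layering a car impact with an occasional York car-hit sound, presumably the horn of the cue's name.
# A mixer combines a boosted random pick of four medium car impacts with a briefly delayed random branch.
# That branch has three waves among ten equally weighted slots, so empty slots usually leave it silent.
# It uses medium-object demolition attenuation and limits concurrency by preventing new plays.
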SoundCue /Game/Sounds/Fort_Vehicles/Honks_Alarms/York_Vehicle_Impact_Horn_Cue
root=Mixer_0
nodes [
 {id=Delay_0, children [Random_1], properties {delayMax=0.05, delayMin=0.01}}
 {id=Mixer_0, children [Random_2,Delay_0], properties {inputVolume [1.25,1]}}
 {
  id=Random_1
  children [WavePlayer_0,null,null,null,WavePlayer_4,null,null,WavePlayer_5,null,null]
  properties {weights [1,1,1,1,1,1,1,1,1,1]}
 }
 {id=Random_2, children [WavePlayer_6,WavePlayer_7,WavePlayer_8,WavePlayer_9], properties {weights [1,1,1,1]}}
 {id=WavePlayer_0, properties {soundWaveAssetPtr=&/York_Car_Hit_03}}
 {id=WavePlayer_4, properties {soundWaveAssetPtr=&/York_Car_Hit_04}}
 {id=WavePlayer_5, properties {soundWaveAssetPtr=&/York_Car_Hit_05}}
 {id=WavePlayer_6, properties {soundWaveAssetPtr=/Game/Sounds/Fort_Impact_Sounds/Demolition/Car/Car_Impact_Med1}}
 {id=WavePlayer_7, properties {soundWaveAssetPtr=/Game/Sounds/Fort_Impact_Sounds/Demolition/Car/Car_Impact_Med2}}
 {id=WavePlayer_8, properties {soundWaveAssetPtr=/Game/Sounds/Fort_Impact_Sounds/Demolition/Car/Car_Impact_Med3}}
 {id=WavePlayer_9, properties {soundWaveAssetPtr=/Game/Sounds/Fort_Impact_Sounds/Demolition/Car/Car_Impact_Med4}}
]
properties {
 attenuationSettings=/Game/Sounds/Attenuation/Fort_Demolition_Hit_Medium_Sized_Object_Att
 bOverrideConcurrency=true
 maxDistance=4000
 soundClassObject=&Fort_Impacts
 volumeMultiplier=1
 concurrencyOverrides {maxCount=4, resolutionRule=PreventNew}
}
